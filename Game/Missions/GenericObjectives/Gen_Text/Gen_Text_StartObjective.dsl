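; Gen_Text_StartObjective is a parameter object, apparently used to start a generic text-based mission objective.
; It stores the objective text, a gameplay tag for the displayed reward and a Slate brush for the displayed icon.
; SetParams writes all three values and returns self for inline setup or chaining, while BreakParams returns them together.
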(fn BreakParams
  (return (Variables|Default|GetObjectiveText) (Variables|Default|GetDisplayedRewardTag) (Variables|Default|GetDisplayedIcon)))

(fn SetParams
  (params (ObjectiveText :type Text) (DisplayedRewardTag :type Gameplay%20Tag%20Structure) (DisplayedIcon :type Slate%20Brush%20Structure))
  (Variables|Default|SetObjectiveText ObjectiveText)
  (Variables|Default|SetDisplayedRewardTag DisplayedRewardTag)
  (Variables|Default|SetDisplayedIcon DisplayedIcon)
  (return self))
