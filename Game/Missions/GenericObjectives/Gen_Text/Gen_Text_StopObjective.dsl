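; Parameter object in the generic text objective mission set, apparently for stopping an objective, that carries a single objective status value.
; SetParams stores an EFortObjectiveStatus, defaulting to Created, in the ObjectiveStatus variable and returns the object itself.
; BreakParams reads that stored status back out, so the pair acts as a pack and unpack wrapper with no other logic.
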(fn BreakParams
  (return (Variables|Default|GetObjectiveStatus)))

(fn SetParams
  (params (ObjectiveStatus :type EFortObjectiveStatus%20Enum :default "Created"))
  (Variables|Default|SetObjectiveStatus ObjectiveStatus)
  (return self))
